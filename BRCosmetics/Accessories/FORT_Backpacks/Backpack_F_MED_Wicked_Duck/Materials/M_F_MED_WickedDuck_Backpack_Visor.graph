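# Visor material for the Wicked Duck backpack that shows glowing, blinking, glitching eyes from the red channel of a mask texture.
# A timer periodically jumps the UVs half a tile in V to blink, while an optional switch adds intermittent hash-driven horizontal glitch distortion.
# Scrolling sine scanlines, weighted by view facing, modulate the tinted emissive strength.
# Attributes then pass through the shared character PostFX and cinematic functions.
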Material /BRCosmetics/Accessories/FORT_Backpacks/Backpack_F_MED_Wicked_Duck/Materials/M_F_MED_WickedDuck_Backpack_Visor
graph {
 properties {
  bUseMaterialAttributes=true
  bUsedWithMeshDeformer=true
  bUsedWithSkeletalMesh=true
  pixelDepthOffsetMode=PDOM_Legacy
  usedShadingModels=Default Lit
 }
 nodes [
  n1 = ComponentMask(None=TextureCoordinate()) {r=true}
  Add(A=@n1, B=Multiply(A=Time(), B=ScalarParameter() {defaultValue=-0.1, parameterName=RB_GlitchLineSpeed}))
  n2 = MaterialFunctionCall() {materialFunction=&BaseFN_CharacterFunction_PreFX}
  n3 = ScalarParameter() {defaultValue=0.2, parameterName=RB_EyeGlowStrength_Min}
  n4 = ScalarParameter() {defaultValue=2, parameterName=RB_EyeGlowStrength_Max}
  n5 = CameraVectorWS()
  n6 = VertexNormalWS()
  n7 = TextureCoordinate()
  n8 = MaterialFunctionCall(Y=Divide(A=Time()) {constB=-6}) {materialFunction=&MakeFloat2}
  n9 = Multiply(A=ComponentMask(None=Add(A=@n7, B=@n8.Result)) {g=true}, B=ScalarParameter() {defaultValue=20, parameterName=RB_ScanLine_Tiling})
  n10 = Clamp(None=Multiply(A=DotProduct(A=@n5, B=@n6), B=Sine(None=@n9)))
  n11 = Ceil(None=Multiply(A=@n1, B=ScalarParameter() {defaultValue=35, parameterName=RB_GlitchLine_Tiling}))
  n12 = ComponentMask(None=MaterialFunctionCall(Seed=Time(), Float=ComponentMask(None=@n11) {g=true}, Vec2=@n11) {materialFunction=&MF_Hash}.Vec2) {r=true}
  n13 = ConstantBiasScale(None=Multiply(A=@n12, B=@n12)) {bias=-0.5, scale=2}
  n14 = ComponentMask(None=TextureCoordinate()) {r=true}
  n15 = Add(A=Time(), B=LinearInterpolate(Alpha=Round(None=@n14)) {constB=0.2})
  n16 = ScalarParameter() {defaultValue=0.05, parameterName=RB_GlitchFrequency}
  n17 = ConstantBiasScale(None=MaterialFunctionCall(Float=Time()) {materialFunction=&MF_Hash}.Float)
  n18 = Subtract(
   A=Frac(None=Multiply(A=@n15, B=@n16) {constB=0.05})
   B=LinearInterpolate(
    A=ScalarParameter() {defaultValue=0.8, parameterName=RB_GlitchThreshold_Min}
    B=ScalarParameter() {defaultValue=0.9, parameterName=RB_GlitchThreshold_Max}
    Alpha=Round(None=@n17)
   )
  )
  n19 = Multiply(A=ScalarParameter() {defaultValue=0.1, parameterName=RB_GlitchDistortion_Amt}, B=Saturate(None=Ceil(None=@n18)))
  n20 = TextureCoordinate()
  n21 = Add(
   A=MaterialFunctionCall(X=Multiply(A=@n13, B=@n19)) {materialFunction=&MakeFloat2}.Result
   B=@n20
  )
  n22 = ScalarParameter() {defaultValue=6, parameterName=Blink Period}
  n23 = Fmod(A=Time(), B=@n22)
  n24 = Subtract(A=@n22, B=ScalarParameter() {defaultValue=0.2, parameterName=Blink Duration})
  n25 = Constant2Vector() {g=0.5}
  n26 = Constant2Vector()
  n27 = TextureSampleParameter2D(UVs=Add(
    A=StaticSwitchParameter(True=@n21, False=@n20) {defaultValue=true, parameterName=Add Glitch}
    B=If(A=@n23, B=@n24, "A > B"=@n25, "A == B"=@n26, "A < B"=@n26)
   )) {
   parameterName=Eyes Texture
   texture=/BRCosmetics/Accessories/FORT_Backpacks/Backpack_F_MED_Wicked_Duck/Textures/T_F_MED_Wicked_Duck_Backpack_Visor_Masks
  }
  n28 = Multiply(A=VectorParameter() {parameterName=Emissive Color, defaultValue=<rgba 0 1 0.9047 1>}.RGB, B=@n27.R)
  n29 = ScalarParameter() {defaultValue=1, parameterName=Emissive Intensity}
  n30 = SetMaterialAttributes!(
   MaterialAttributes
   Base Color
   Emissive Color=Multiply(A=LinearInterpolate(A=@n3, B=@n4, Alpha=@n10), B=Multiply(A=@n28, B=@n29))
   Specular=Constant() {r=0.5}
   Roughness=Constant() {r=1}
  ) {attributeSetTypes [
    #1
    #2
    #3
    #4
   ]}
  n31 = MaterialFunctionCall(AO=@n2.AO, Distance Blend=@n2.AO, Fresnel=@n2.AO, Gmap=@n2.AO, Pre Skinned Local Position=@n2.AO, Material=@n30) {materialFunction=&BaseFN_CharacterFunction_PostFX}
  MaterialFunctionCall(In=@n31.Result) {materialFunction=&CINE_MF_CharPostFX}
 ]
}
