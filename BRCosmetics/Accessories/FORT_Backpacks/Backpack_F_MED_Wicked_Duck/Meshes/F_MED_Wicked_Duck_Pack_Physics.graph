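# Physics asset for the Wicked Duck backpack cosmetic, providing bodies and joints for the duck's legs and feet.
# Five capsules cover the bones: one large capsule for the pack and smaller ones for each leg and foot.
# Constraints form two chains from pack to leg to foot.
# Each leg joint locks twist and swing2, leaving swing1 as its hinge axis, while the foot joints lock no angular axis.
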PhysicsAsset /BRCosmetics/Accessories/FORT_Backpacks/Backpack_F_MED_Wicked_Duck/Meshes/F_MED_Wicked_Duck_Pack_Physics
bodies {
 foot_l [{shapeType=Capsule, length=0.75, radius=1.5, center=<0.4699 -0.805 -0.4024>, rotation=<rot 0 0 -89.9996>}]
 foot_r [{shapeType=Capsule, length=0.9573, radius=1.3991, center=<-0.0645 0.3829 -0.0361>, rotation=<rot 0 9.5665 -84.687>}]
 leg_l [{shapeType=Capsule, length=1.6744, radius=0.9402, center=<1.8253 2.1283 0.024>, rotation=<rot 0 -89.746 -89.7069>}]
 leg_r [{shapeType=Capsule, length=1.6744, radius=0.9624, center=<-0.4874 -1.2839 -0.0349>, rotation=<rot 0 -89.8961 -89.912>}]
 pack [{shapeType=Capsule, length=1.3762, radius=7.4698, center=<-0.047 -9.7289 -0.2784>, rotation=<rot 16.2507 -90 -2.7804>}]
}
constraints [
 {bone1Name=leg_l, bone2Name=pack, swing1LimitDegrees=10, swing2LimitDegrees=15, swing2Motion=Locked, twistLimitDegrees=45, twistMotion=Locked}
 {bone1Name=foot_l, bone2Name=leg_l, swing1LimitDegrees=15, swing2LimitDegrees=5, twistLimitDegrees=1}
 {bone1Name=leg_r, bone2Name=pack, swing1LimitDegrees=10, swing2LimitDegrees=15, swing2Motion=Locked, twistLimitDegrees=45, twistMotion=Locked}
 {bone1Name=foot_r, bone2Name=leg_r, swing1LimitDegrees=15, swing2LimitDegrees=5, twistLimitDegrees=1}
]
